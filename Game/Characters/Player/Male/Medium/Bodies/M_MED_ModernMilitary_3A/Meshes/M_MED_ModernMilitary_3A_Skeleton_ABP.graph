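# Post-process-style AnimBlueprint for the M_MED_ModernMilitary_3A male player body, following the owner's skeletal mesh and adding secondary motion.
# It copies that mesh's pose, corrects elbow pads with pose drivers, then runs a Control Rig fed pistol, rifle and emote alphas.
# A cached pose shrinks the head bone when the playlist uses custom character parts, then feeds pelvis-space RigidBody physics.
# The event graph derives alphas from weapon animation type and emote state.
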AnimBlueprint /Game/Characters/Player/Male/Medium/Bodies/M_MED_ModernMilitary_3A/Meshes/M_MED_ModernMilitary_3A_Skeleton_ABP
target_skeleton=&/M_MED_ModernMilitary_3A_Skeleton
graphs [
 AnimGraph {nodes [
   Anim Dynamics(Alpha=1, ExternalForce="0,0,0") {
    bGravityOverrideInSimSpace=true
    alphaBoolBlend {blendOption=Linear}
    physicsBodyDefinitions [{
      sphereCollisionRadius=0
      boundBone {boneName=None}
      boxExtents=<1 1 1>
      constraintSetup {angularLimitsMax=<0 0 0>, angularLimitsMin=<0 0 0>, angularTarget=<0 0 0>}
      localJointOffset=<0 0 0>
     }]
    boundBone=~
   }
   Component To Local()
   Local To Component()
   n1 = "Transform (Modify) Bone - Bone: head"!(ComponentPose=Local To Component(LocalPose=Use cached pose 'PreScale Cache'()), Alpha=1) -> (Pose) {scaleMode=BMM_Replace, alphaBoolBlend {blendOption=Linear}, boneToModify {boneName=head}, scale=<0.01 0.01 0.01>}
   n2 = Component To Local(ComponentPose=@n1)
   n3 = Use cached pose 'PreScale Cache'()
   n4 = Get GetPlayerListUsesCustomCharacterParts() -> (GetPlayerListUsesCustomCharacterParts:bool)
   n5 = RigidBody(
    ComponentPose=Local To Component(LocalPose=Blend Poses by bool(BlendPose_0=@n2, BlendPose_1=@n3, BlendTime_0=0, BlendTime_1=0, bActiveValue=@n4) {blendType=Linear})
    Alpha=0.6
   ) {
    bClampLinearTranslationLimitToRefPose=true
    bOverrideWorldGravity=true
    overridePhysicsAsset=&/M_MED_ModernMilitary_3A_Physics
    simulationSpace=BaseBoneSpace
    alphaBoolBlend {blendOption=Linear}
    baseBoneRef {boneName=pelvis}
    componentAppliedLinearAccClamp=<2000 2000 4000>
    componentLinearAccScale=<0.5 0.3 0.5>
    componentLinearVelScale=<0.3 0.1 0.3>
    overrideWorldGravity {z=-980}
   }
   Output Pose(Result=Component To Local(ComponentPose=@n5)) {layerGroup=None}
   n6 = Pose Driver - Source: lowerarm_l!(SourcePose:Pose Link Structure=Copy Pose From Mesh(SourceMeshComponent=Get mesh_to_copy() -> (mesh_to_copy:Skeletal Mesh Component Object Reference))) -> (Pose:Pose Link Structure) {
    is=title AnimGraphNode_PoseDriver
    poseAsset=/Game/Characters/Player/Male/Medium/Bodies/M_MED_ModernMilitary_3A/Meshes/pose_driver/M_MED_ModernMilitary_3A_pose
    onlyDriveBones [{boneName=elbow_pad_l}]
    poseTargets [{
      bApplyCustomCurve=false
      bIsHidden=false
      distanceMethod=DefaultMethod
      drivenName=elbow_l
      functionType=DefaultFunction
      targetScale=1
      boneTransforms [{targetRotation=<rot -3.6203 128.4412 -2.8693>, targetTranslation=<26.6514 0 0>}]
      customCurve {defaultValue=3.4028234663852886e+38, postInfinityExtrap=RCCE_Constant, preInfinityExtrap=RCCE_Constant, keys []}
      targetRotation=<rot 0 0 0>
     }]
    rBFParams {radius=49.9661}
    sourceBones [{boneName=lowerarm_l}]
   }
   n7 = Pose Driver - Source: lowerarm_r!(SourcePose:Pose Link Structure=@n6) -> (Pose:Pose Link Structure) {
    is=title AnimGraphNode_PoseDriver
    poseAsset=&M_MED_ModernMilitary_3A_pose
    onlyDriveBones [{boneName=elbow_pad_r}]
    poseTargets [{
      bApplyCustomCurve=false
      bIsHidden=false
      distanceMethod=DefaultMethod
      drivenName=elbow_r
      functionType=DefaultFunction
      targetScale=1
      boneTransforms [{targetRotation=<rot -3.624 128.3674 -2.8647>, targetTranslation=<-26.6515 0 0>}]
      customCurve {defaultValue=3.4028234663852886e+38, postInfinityExtrap=RCCE_Constant, preInfinityExtrap=RCCE_Constant, keys []}
      targetRotation=<rot 0 0 0>
     }]
    rBFParams {function=Linear, radius=49.9293}
    sourceBones [{boneName=lowerarm_r}]
   }
   n8 = Control Rig!(
    Source=@n7
    Alpha
    Is Pistol Alpha:double=Get Is Pistol Alpha() -> (Is Pistol Alpha:double)
    Is Rifle Alpha:double=Get Is Rifle Alpha() -> (Is Rifle Alpha:double)
    Is Emote Alpha:double=Get Is Emote Alpha() -> (Is Emote Alpha:double)
   ) -> (Pose) {
    alphaBoolBlend {blendOption=Linear}
    controlRigAssetReference {blueprintRigClass=&/M_MED_ModernMilitary_3A_CtrlRig.M_MED_ModernMilitary_3A_CtrlRig_C}
   }
   Save cached pose 'PreScale Cache'(Pose=@n8) {cacheName=PreScale Cache}
   Sequence Player 'M_MED_ModernMilitary_3A_pose_anim'() {
    playRate=0
    playRateBasis=0
    sequence=/Game/Characters/Player/Male/Medium/Bodies/M_MED_ModernMilitary_3A/Meshes/pose_driver/M_MED_ModernMilitary_3A_pose_anim
   }
  ]}
 EventGraph {nodes [
   n1 = Cast To FortPlayerPawn(Object=GetPlayerPawn!(PlayerIndex:int=0) -> (ReturnValue:Pawn Object Reference))
   > PrintString!(
    execute
    InString="To String (Integer)"(InInt=GetSignificanceLOD!(self:Fort Player Pawn Object Reference=@n1) -> (ReturnValue:int=0))
    bPrintToScreen
    bPrintToLog
    TextColor
    Duration
    Key
   ) -> (then)
   Event BlueprintUpdateAnimation()
   > n2 = Sequence!(execute) -> (then_0, then_1, then_2:Exec)
   > Set mesh_to_copy(mesh_to_copy:Skeletal Mesh Component Object Reference=GetOwnerSkeletalMeshComponent()) -> (Output_Get:Skeletal Mesh Component Object Reference)
   n3 = Sequence!(execute=@n2.then_1) -> (then_0, then_1, then_2, then_3:Exec, then_4:Exec)
   > n4 = Cast To FortPlayerAnimInstance(Object=GetAnimInstance(self=GetOwnerSkeletalMeshComponent()))
   > Set isPistol(isPistol=false)
   > Set isRifle(isRifle:bool=false) -> (Output_Get:bool)
   > n5 = Switch on EFortWeaponCoreAnimation!(execute, Selection:EFortWeaponCoreAnimation Enum=Get WeaponCoreAnim(self=@n4)) -> (
    Melee:Exec
    Pistol:Exec
    Shotgun:Exec
    PaperBlueprint:Exec
    Rifle:Exec
    MeleeOneHand:Exec
    MachinePistol:Exec
    RocketLauncher:Exec
    GrenadeLauncher:Exec
    GoingCommando:Exec
    AssaultRifle:Exec
    TacticalShotgun:Exec
    SniperRifle:Exec
    TrapPlacement:Exec
    ShoulderLauncher:Exec
    AbilityDecoTool:Exec
    Crossbow:Exec
    C4:Exec
    RemoteControl:Exec
    DualWield:Exec
    AR_BullPup:Exec
    AR_ForwardGrip:Exec
    MedPackPaddles:Exec
    SMG_P90:Exec
    AR_DrumGun:Exec
    Consumable_Small:Exec
    Consumable_Large:Exec
    Balloon:Exec
    MountedTurret:Exec
    CreativeTool:Exec
    ExplosiveBow:Exec
    AshtonIndigo:Exec
    AshtonChicago:Exec
    MeleeDualWield:Exec
    Unarmed:Exec
   )
   Set isRifle(execute=[@n5.AssaultRifle, @n5.Rifle, @n5.RocketLauncher, @n5.Shotgun, @n5.SniperRifle, @n5.TacticalShotgun], isRifle=true)
   Set isPistol(execute=@n5.Pistol, isPistol=true)
   Set Is Rifle Alpha(execute=@n3.then_1, Is Rifle Alpha:double="To Float (Boolean)"(InBool=Get isRifle() -> (isRifle:bool))) -> (Output_Get:double)
   Set Is Pistol Alpha(execute=@n3.then_2, Is Pistol Alpha:double="To Float (Boolean)"(InBool=Get isPistol() -> (isPistol:bool))) -> (Output_Get:double)
   Set Is Emote Alpha(execute=@n2.then_2, Is Emote Alpha:double="To Float (Boolean)"(InBool=Get bIsPlayingEmote())) -> (Output_Get:double)
   Event BlueprintInitializeAnimation()
   > Set GetPlayerListUsesCustomCharacterParts(GetPlayerListUsesCustomCharacterParts:bool=GetPlaylistUsesCustomCharacterParts!() -> (ReturnValue:bool)) -> (Output_Get:bool)
   Get bIsSkyDiving()
  ]}
]
variables [
 GetPlayerListUsesCustomCharacterParts {type=boolean}
 Is Emote Alpha {type=number}
 Is Pistol Alpha {type=number}
 Is Rifle Alpha {type=number}
 isPistol {type=boolean}
 isRifle {type=boolean}
 mesh_to_copy {type=SkeletalMeshComponent}
]
